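; Master Blueprint for the Creative shooting target device, which pops up, hops, falls when shot and awards team-filtered score.
; Player options and minigame state drive setup, with timers, timelines and a proximity trigger moving the target.
; Server damage handling detects weak-point hits, applies critical multipliers, and clamps health to avoid death, multicasting knockdowns.
; Hops replicate through OnRep variables, and state changes broadcast gameplay events, or notify listeners when track-attached.
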(event Custom|SpawnWeakpointFX
  (params (ImpactTransform_Location :type Vector :default "0, 0, 0") (ImpactTransform_Rotation :type Rotator :default "0, 0, 0") (ImpactTransform_Scale :type Vector :default "0, 0, 0"))
  ;[1 1 1] Spawn Weakpoint FX
  (Networking|CanExecuteCosmeticEvents
    (:True
      (Effects|SpawnEmitteratLocation "/Game/Effects/Fort_Effects/Effects/Creative/Props/P_TargetImpactBullseye.P_TargetImpactBullseye" ImpactTransform_Location ImpactTransform_Rotation ImpactTransform_Scale true "AutoRelease"))))

(event Custom|PickMeshFromSettings
  ;[1 1 1] Setting and loading the correct mesh and turning on collision for the correct weak point for the target
  (bind _loaded_object (Creative|Loading|AsyncLoadEditorSafe (Utilities|Array|Get(acopy) (Variables|Default|GetTargetMeshesArray) (Variables|PlayerOptions|GetTargetMeshSetting)))
    (:Loaded
      (bind _asstatic_mesh (Utilities|Casting|CastToStaticMesh _loaded_object))
      (Components|StaticMesh|SetStaticMesh (Variables|Default|GetTargetMesh) _asstatic_mesh)
      (Physics|SetPhysicalMaterialOverride (Variables|Default|GetTargetMesh) "/Game/FortressPhysicalMaterials/Metal.Metal"))))

(event Custom|Reset%20Target
  (Utilities|FlowControl|Sequence
    (:then_0
      (CallFunction|ClearAllTimers)
      ;[1 1 1] Setting and loading the correct mesh and turning on collision for the correct weak point for the target
      (CallFunction|RestoretoMaxHealth)
      (Variables|Default|SetNumResets 0)
      (Variables|Default|SetNumPopUps 0) ; Ensures NumPopups is reset to -1 at game start.  This is so the Target Pops up at least one timeif it is set to up
      (CallFunction|InitializeSettings))
    (:then_1
      (CallFunction|ActivateTimers))))

(event Custom|Knock%20Down%20Audio
  (params (Knocked%20Down? :type Boolean))
  ;[1 1 1] Sound for when target gets knocked down and comes backup
  (Audio|PlaySoundatLocation (select Knocked%20Down? "/Game/Sounds/Creative/Toys/ShootingTargets/Target_Down_Cue.Target_Down_Cue" "/Game/Sounds/Creative/Toys/ShootingTargets/Target_FoldUp_Cue.Target_FoldUp_Cue") (Transformation|GetActorLocation)))

(event MinigameEndedDelegate(FortMinigameProgress)
  (CallFunction|SetEnabledStateCheck false)
  (CallFunction|RestoretoMaxHealth)
  (Variables|Default|SetNumResets 0)
  (Variables|Default|SetNumPopUps 0) ; Ensures NumPopups is reset to -1 at game start.  This is so the Target Pops up at least one timeif it is set to up
  (CallFunction|InitializeSettings))

(event MinigameStartedDelegate(FortMinigameProgress)
  (CallFunction|SetEnabledStateCheck true)
  (Variables|Default|SetNumResets 0)
  (Variables|Default|SetNumPopUps 0) ; Ensures NumPopups is reset to -1 at game start.  This is so the Target Pops up at least one timeif it is set to up
  (CallFunction|InitializeSettings))

(event EventBeginPlay
  ;[1 1 1] Helps this all work during non-island pie sessions
  (Class|FortAsyncActionAuthorityMatchReady|MatchReadyAsync
    (:OnReady
      ;[1 1 1] Attach to minigame here as targets spawned by tracks do not attach automatically
      (Class|FortMinigameLogicComponent|TryAttachtoMinigame (Variables|Default|GetFortMinigameProgress))
      (CallFunction|InitializeSettings))))

(event Custom|Impact%20Audio
  (params (Check%20for%20Positive%20Score :type Integer) (Instigator :type Controller%20Object%20Reference))
  (bind _returnvalue (Pawn|IsLocalController Instigator))
  (bind _returnvalue_2 (Transformation|GetActorLocation))
  (bind _weakpointhit (Variables|Default|GetWeakPointHit))
  ;[1 1 1] Sound For when target get shot; Was it a weakpoint or not
  (if (>= Check%20for%20Positive%20Score 0)
    (Audio|PlaySoundatLocation (select _weakpointhit "/Game/Sounds/Creative/Toys/ShootingTargets/Target_Impact_Bullseye_Cue.Target_Impact_Bullseye_Cue" "/Game/Sounds/Creative/Toys/ShootingTargets/Target_Impact_Cue.Target_Impact_Cue") _returnvalue_2)
    (if (and _weakpointhit _returnvalue)
      (Audio|PlaySound2D "/Game/Sounds/Creative/Toys/ShootingTargets/Target_Impact_Success_2D_Cue.Target_Impact_Success_2D_Cue"))
    (else
      (Audio|PlaySoundatLocation "/Game/Sounds/Creative/Toys/ShootingTargets/Target_Impact_Friendly_Cue.Target_Impact_Friendly_Cue" _returnvalue_2)
      (if _returnvalue
        (Audio|PlaySound2D "/Game/Sounds/Creative/Toys/ShootingTargets/Target_Error_Cue.Target_Error_Cue")))))

(event EventPostFinishSpawning
  (|CallParentFunction)
  (CallFunction|InitializeSettings))

(event EventOnDamageServer
  (params (Damage :type Float%20%28single-precision%29) (DamageTags :type Gameplay%20Tag%20Container%20Structure%20%28by%20ref%29) (Momentum :type Vector :default "0, 0, 0") (HitInfo :type Hit%20Result%20Structure%20%28by%20ref%29) (InstigatedBy :type Controller%20Object%20Reference) (DamageCauser :type Actor%20Object%20Reference) (EffectContext :type Gameplay%20Effect%20Context%20Handle%20Structure))
  ;[1 1 1] React to damage
  (Variables|Default|SetActorLocationAtMomentofHit (Transformation|GetWorldLocation (Variables|Default|GetTargetMesh)))
  (if (CallFunction|TargetEnabled)
    (bind (_hitcomponent _returnvalue) (CallFunction|WeakPointWasHit HitInfo))
    (Variables|Default|SetWeakPointHit _returnvalue)
    (if (== (Variables|PlayerOptions|GetStartingHealthSetting) (Variables|Default|GetInfiniteHealthIndex))
      (Health|SetHealth (Health|GetMaxHealth))
      (else
        _))
    (bind _returnvalue_2 (CallFunction|TargetShouldFall))
    (if _returnvalue_2
      (if (<= (Health|GetHealth) 0.0)
        (Health|SetHealth (* (Health|GetMaxHealth) 0.01)) ; Don't let it actually die.  Set health to 1% so bar looks mostly empty
        (else
          _))
      (if (Variables|Default|GetTargetIsUp)
        (CallFunction|SelectResetorPopupTime true)
        (CallFunction|RPCMulticastKnockdownTarget true)
        (if (Variables|Default|GetWeakPointHit)
          (CallFunction|HandleHitTransmitting true true InstigatedBy)
          (else
            (CallFunction|HandleHitTransmitting false true InstigatedBy)))
        (bind _asfort_player_controller (Utilities|Casting|CastToFortPlayerController InstigatedBy))
        (bind _scoreawarded (CallFunction|AwardScore _asfort_player_controller))
        (else
          _))
      (CallFunction|ImpactAudio _scoreawarded InstigatedBy)
      (Default|CallTargetHitDispatcher InstigatedBy)
      (else
        (if (Variables|Default|GetTargetIsUp)
          (CallFunction|HandleHitTransmitting (Variables|Default|GetWeakPointHit) false InstigatedBy)
          (else
            _))))))

(event Custom|Target%20Movement%20Audio
  (params (Jump%20Up? :type Boolean))
  ;[1 1 1] Play sound when target moves
  (Audio|PlaySoundatLocation (select Jump%20Up? "/Game/Sounds/Creative/Toys/ShootingTargets/Target_Raise_Cue.Target_Raise_Cue" "/Game/Sounds/Creative/Toys/ShootingTargets/Target_Lower_Cue.Target_Lower_Cue") (Transformation|GetActorLocation)))

(event Custom|Knockdown%20Target
  (params (WasShot :type Boolean))
  ;[1 1 1] Knock down target
  (Variables|Default|SetTargetShot WasShot)
  (Utilities|Time|ClearandInvalidateTimerbyHandle (Variables|Default|GetHideTargetTimer))
  (CallFunction|ClearBunnyHopTimer)
  (CallFunction|KnockDownAudio true)
  (CallFunction|UpdateHealthBarVisibility false)
  (Variables|Default|SetTargetIsUp false)
  (if (== (Variables|PlayerOptions|GetHingeLocationSetting) 0)
    (Variables|Default|SetRotated_Piece (Variables|Default|GetSM_Target_StandShort))
    (else
      (Variables|Default|SetRotated_Piece (Variables|Default|GetTargetMesh))))
  (CallFunction|SetAimAssistEnabled false)
  (bind (_direction _rotation) (Timeline|KnockDownTimeline PlayFromStart)
    (:Update
      (Transformation|SetRelativeRotation (Variables|Default|GetRotated_Piece) (Math|Rotator|MakeRotator 0.0 _rotation) false false))
    (:Finished
      (if (not (Variables|Default|GetTargetShot))
        (Networking|SwitchHasAuthority
          (:Authority
            (if (CallFunction|ReadytoActivate)
              (CallFunction|HandleTransmitter 0 "NewEnumerator5")
              (else
                _))
            (Default|CallTargetKnockedDownDispatcher)
            (Components|Activation|Activate (Variables|Default|GetP_TargetImpact_Falldown) false)
            (if (!= (Variables|Default|GetPopup%20Time) 0.0)
              (bind _returnvalue (Utilities|Time|SetTimerbyEvent (delegate Custom|PopUpTargetTimerEvent) (Variables|Default|GetPopup%20Time)))
              (Variables|Default|SetPopUpDelayTimer _returnvalue)))
          (:Remote
            _))
        (else
          _)))))

(event Custom|ResetFromSettings
  ;[1 1 1] Knock down target
  (if (!= (Variables|Default|GetPopup%20Time) 0.0)
    (bind _returnvalue (Utilities|Time|SetTimerbyEvent (delegate Custom|PopUpTargetTimerEvent) (Variables|Default|GetPopup%20Time)))
    (Variables|Default|SetPopUpDelayTimer _returnvalue)))

(event Custom|EnableTarget
  ;[1 1 1] Enable/Disable Target
  (Class|CreativeEnabledComponent|SetEnabled (Variables|Default|GetCreative_Enabled_Component) true)
  (Variables|Default|SetbJustEnabled true)
  (CallFunction|SelectResetorPopupTime false)
  (CallFunction|PopUpTarget false)
  (CallFunction|ActivateTimers))

(event Custom|RPC_Multicast_KnockdownTarget
  (params (WasShot :type Boolean))
  ;[1 1 1] Knock down target
  (CallFunction|KnockdownTarget WasShot))

(event Custom|PopUpTargetTimerEvent
  ;[1 1 1] Knock down target
  (if (CallFunction|ReadytoActivate)
    (CallFunction|PopUpTarget false (Variables|Default|GetTargetShot))))

(event Custom|DisableTarget
  ;[1 1 1] Enable/Disable Target
  (Class|CreativeEnabledComponent|SetEnabled (Variables|Default|GetCreative_Enabled_Component) false))

(event Custom|TargetKnockedDownDispatcher_Event
  ;[1 1 1] Knock down target
  (Development|PrintString "Target Knockdoiwn event fired"))

(event Custom|EnableFromEvent
  (params (Instigator :type Controller%20Object%20Reference))
  ;[1 1 1] Enable/Disable Target
  (CallFunction|EnableTarget))

(event Custom|PopUpTarget
  (params (MinigameJustStartedOrStopped :type Boolean) (WasShot :type Boolean))
  (bind _rotated_piece (Variables|Default|GetRotated_Piece))
  ;[1 1 1] Put Target back up again
  (if (CallFunction|CanPopUp MinigameJustStartedOrStopped WasShot)
    (CallFunction|RestoretoMaxHealth)
    (if MinigameJustStartedOrStopped
      _
      (elif (not WasShot)
        (Math|Integer|IncrementInt (Variables|Default|GetNumPopUps)
          (:%20%20
            _))
        (else
          (Math|Integer|IncrementInt (Variables|Default|GetNumResets)
            (:%20%20
              _)))))
    (Variables|Default|SetTargetIsUp true) ; Before timeline so hits while\nit is coming up will register
    (CallFunction|ClearPlayedDying)
    (CallFunction|UpdateHealthBarVisibility true)
    (CallFunction|SetBunnyHopTimer)
    (CallFunction|KnockDownAudio false)
    (CallFunction|SetAimAssistEnabled true)
    (bind (_direction _rotation) (Timeline|PopUpTimeline PlayFromStart)
      (:Update
        (Utilities|IsValid _rotated_piece
          (:Is%20Valid
            (Transformation|SetRelativeRotation _rotated_piece (Math|Rotator|MakeRotator 0.0 _rotation) false false))))
      (:Finished
        (if (not (CallFunction|ReadytoActivate))
          (Networking|SwitchHasAuthority
            (:Authority
              (CallFunction|RPCMulticastKnockdownTarget false)))
          (else
            ;[1 1 1] Handles whether the target should automatically go down and how long it should wait to do so.
            (Networking|SwitchHasAuthority
              (:Authority
                (if (not (Variables|Default|GetbJustEnabled))
                  (CallFunction|HandleTransmitter 0 "NewEnumerator4")
                  (else
                    (Variables|Default|SetbJustEnabled false)))
                (Default|CallTargetPopUpDispatcher)
                (if (not (== (Variables|PlayerOptions|GetTimeBeforeHidingSetting) 0.0))
                  (bind _output_time (CallFunction|SelectHideTime))
                  (CallFunction|SetHideTargetTimer _output_time)))
              (:Remote
                _))))))))

(event Custom|DisableFromEvent
  (params (Instigator :type Controller%20Object%20Reference))
  ;[1 1 1] Enable/Disable Target
  (CallFunction|DisableTarget))

(event Custom|SetHideTargetTimer
  (params (TimerDuration :type Float%20%28double-precision%29))
  ;[1 1 1] Timer for hiding target.  Needs to be set from multiple locations so made an event to call this block
  (if (CallFunction|ReadytoActivate)
    (bind _returnvalue (Utilities|Time|SetTimerbyEvent (delegate Custom|HideTargetTimerEvent) TimerDuration))
    (Variables|Default|SetHideTargetTimer _returnvalue)))

(event OnComponentBeginOverlap(ProximityTrigger)
  (params (OverlappedComponent :type Primitive%20Component%20Object%20Reference) (OtherActor :type Actor%20Object%20Reference) (OtherComp :type Primitive%20Component%20Object%20Reference) (OtherBodyIndex :type Integer) (bFromSweep :type Boolean) (SweepResult :type Hit%20Result%20Structure%20%28by%20ref%29))
  ;[1 1 1] Handles Proximity causing target to pop up
  (if (Variables|Default|GetWaitingForProximity)
    (Variables|Default|SetWaitingForProximity false)
    (Networking|SwitchHasAuthority
      (:Authority
        (bind _result (CallFunction|ShouldTriggerTarget OtherActor)) ; Make sure overlapping actor should activate trigger
        (if _result
          (Utilities|Time|ClearandInvalidateTimerbyHandle (Variables|Default|GetPopUpDelayTimer)) ; If timer is running from a previous knock down clear that timer to allow an instant pop up and no random pop up later
          (if (not (Variables|Default|GetTargetIsUp))
            (CallFunction|PopUpTarget false)))))))

(event Custom|HideTargetTimerEvent
  ;[1 1 1] Timer for hiding target.  Needs to be set from multiple locations so made an event to call this block
  (CallFunction|HideTarget))

(event OnAnyPropertyChanged(ToyOptionsComponent)
  ;[1 1 1] Initialize and update settings
  (Utilities|FlowControl|DelayUntilNextTick)
  (Class|CreativeEnabledComponent|SetEnabledOnPlayMode (Variables|Default|GetCreative_Enabled_Component) (Variables|PlayerOptions|GetEnabledOnMinigameStart))
  (CallFunction|ClearAllTimers)
  (Variables|Default|SetWaitingForProximity (!= (Variables|PlayerOptions|GetProximityDistanceSetting) (Variables|Default|GetProximityOffSetting)))
  (CallFunction|PickMeshfromSettings)
  (CallFunction|SetStartingPosition)
  (CallFunction|SetMaxHealth (Variables|PlayerOptions|GetStartingHealthSetting))
  (CallFunction|UpdateProximitySphereCollision)
  (CallFunction|CheckRandomMinMax)
  (CallFunction|SetLegacyResetDefaults)
  (CallFunction|UpdateHealthBarVisibility true)
  (if (Variables|PlayerOptions|GetEnabledOnMinigameStart)
    (if (CallFunction|ReadytoActivate)
      _)
    (elif (Utilities|Enum|NotEqual(Enum) (Variables|Default|GetCurrentGameState) "InProgress")
      _
      (else
        (Networking|SwitchHasAuthority
          (:Authority
            (if (Variables|Default|GetTargetIsUp)
              ;;[1 1 1] Handles getting targets working again after players change reset delay from "never reset" back to a real time.
              (CallFunction|RPCMulticastKnockdownTarget false)))))))
  (CallFunction|ActivateTimers)
  (if (not (Variables|Default|GetTargetIsUp))
    (CallFunction|ResetfromSettings)))

(event OnPropertiesLoaded(ToyOptionsComponent)
  ;[1 1 1] Initialize and update settings
  (Utilities|FlowControl|DelayUntilNextTick)
  (Class|CreativeEnabledComponent|SetEnabledOnPlayMode (Variables|Default|GetCreative_Enabled_Component) (Variables|PlayerOptions|GetEnabledOnMinigameStart))
  (CallFunction|ClearAllTimers)
  (Variables|Default|SetWaitingForProximity (!= (Variables|PlayerOptions|GetProximityDistanceSetting) (Variables|Default|GetProximityOffSetting)))
  (CallFunction|PickMeshfromSettings)
  (CallFunction|SetStartingPosition)
  (CallFunction|SetMaxHealth (Variables|PlayerOptions|GetStartingHealthSetting))
  (CallFunction|UpdateProximitySphereCollision)
  (CallFunction|CheckRandomMinMax)
  (CallFunction|SetLegacyResetDefaults)
  (CallFunction|UpdateHealthBarVisibility true)
  (if (Variables|PlayerOptions|GetEnabledOnMinigameStart)
    (if (CallFunction|ReadytoActivate)
      _)
    (elif (Utilities|Enum|NotEqual(Enum) (Variables|Default|GetCurrentGameState) "InProgress")
      _
      (else
        (Networking|SwitchHasAuthority
          (:Authority
            (if (Variables|Default|GetTargetIsUp)
              ;;[1 1 1] Handles getting targets working again after players change reset delay from "never reset" back to a real time.
              (CallFunction|RPCMulticastKnockdownTarget false)))))))
  (CallFunction|ActivateTimers)
  (if (not (Variables|Default|GetTargetIsUp))
    (CallFunction|ResetfromSettings)))

(event Custom|ActivateTimers
  ;[1 1 1] Initialize and update settings
  (CallFunction|InitializeHideTargetTimer)
  (CallFunction|CheckifShouldHop))

(event Custom|InitializeSettings
  ;[1 1 1] Initialize and update settings
  (Class|CreativeEnabledComponent|SetEnabledOnPlayMode (Variables|Default|GetCreative_Enabled_Component) (Variables|PlayerOptions|GetEnabledOnMinigameStart))
  (CallFunction|ClearAllTimers)
  (Variables|Default|SetWaitingForProximity (!= (Variables|PlayerOptions|GetProximityDistanceSetting) (Variables|Default|GetProximityOffSetting)))
  (CallFunction|PickMeshfromSettings)
  (CallFunction|SetStartingPosition)
  (CallFunction|SetMaxHealth (Variables|PlayerOptions|GetStartingHealthSetting))
  (CallFunction|UpdateProximitySphereCollision)
  (CallFunction|CheckRandomMinMax)
  (CallFunction|SetLegacyResetDefaults)
  (CallFunction|UpdateHealthBarVisibility true)
  (if (Variables|PlayerOptions|GetEnabledOnMinigameStart)
    (if (CallFunction|ReadytoActivate)
      _)
    (elif (Utilities|Enum|NotEqual(Enum) (Variables|Default|GetCurrentGameState) "InProgress")
      _
      (else
        (Networking|SwitchHasAuthority
          (:Authority
            (if (Variables|Default|GetTargetIsUp)
              ;;[1 1 1] Handles getting targets working again after players change reset delay from "never reset" back to a real time.
              (CallFunction|RPCMulticastKnockdownTarget false)))))))
  (CallFunction|ActivateTimers)
  (if (not (Variables|Default|GetTargetIsUp))
    (CallFunction|ResetfromSettings)))

(event Custom|Hop
  ;[1 1 1] Hop management
  (CallFunction|TargetMovementAudio true)
  (bind (_direction _height) (Timeline|Timeline_2 PlayFromStart)
    (:Update
      (Transformation|SetRelativeLocation (Variables|Default|GetHeightExtender) (Math|Vector|MakeVector 4.0 0.0 _height) false false))
    (:Finished
      (if (CallFunction|ReadytoActivate)
        (if (Variables|Default|GetInJumpUpPosition)
          (Networking|SwitchHasAuthority
            (:Authority
              (CallFunction|SelectHoporDropTime false)
              (CallFunction|HandleTransmitter 0 "NewEnumerator6")))
          (else
            (Networking|SwitchHasAuthority
              (:Authority
                (CallFunction|SelectHoporDropTime true)
                (if (Variables|Default|GetStartOfMatchHopDown)
                  (Variables|Default|SetStartOfMatchHopDown false)
                  (else
                    (CallFunction|HandleTransmitter 0 "NewEnumerator7")))))))))))

(event Custom|DropFromHop
  ;[1 1 1] Hop management
  (CallFunction|TargetMovementAudio false)
  (Timeline|Timeline_2 ReverseFromEnd))

(event Custom|SetHopTimer
  (params (Time :type Float%20%28double-precision%29))
  ;[1 1 1] Timer needs to be set from multiple locations in the graph so made an event to call it
  (bind _returnvalue (Utilities|Time|SetTimerbyEvent (delegate Custom|HopTimerEvent) Time))
  (Variables|Default|SetHopTimer _returnvalue))

(event Custom|HopTimerEvent
  ;[1 1 1] Timer needs to be set from multiple locations in the graph so made an event to call it
  (if (and (Variables|Default|GetTargetIsUp) (not (Variables|Default|GetInJumpUpPosition)))
    (Variables|Default|SetInJumpUpPosition true)))

(event Custom|SetDropHopTimer
  (params (Time :type Float%20%28double-precision%29))
  ;[1 1 1] Timer needs to be set from multiple locations in the graph so made an event to call it
  (bind _returnvalue (Utilities|Time|SetTimerbyEvent (delegate Custom|DropHopTimerEvent) Time))
  (Variables|Default|SetHoppedUpPositionTimer _returnvalue))

(event Custom|DropHopTimerEvent
  ;[1 1 1] Timer needs to be set from multiple locations in the graph so made an event to call it
  (if (and (Variables|Default|GetTargetIsUp) (Variables|Default|GetInJumpUpPosition))
    (Variables|Default|SetInJumpUpPosition false)))

(event MinigameStateChangedDelegate(FortMinigameProgress)
  (params (Minigame :type Fort%20Minigame%20Object%20Reference) (NewMinigameState :type EFortMinigameState%20Enum :default "PreGame"))
  ;[1 1 1] Minigame State Change
  (Variables|Default|SetCurrentGameState NewMinigameState)
  (if (not (or (Utilities|Enum|Equal(Enum) NewMinigameState "PreGame") (Utilities|Enum|Equal(Enum) NewMinigameState "InProgress")))
    (CallFunction|ClearAllTimers)))

(event Custom|ResetWhenReceiving
  ;[0.3658 0 1 0.25] Gameplay Functions
  (if (CallFunction|TargetEnabled)
    (CallFunction|ResetTarget)))

(event Custom|Reset%20Event
  (params (Instigator :type Controller%20Object%20Reference))
  ;[0.3658 0 1 0.25] Gameplay Functions
  (CallFunction|ResetwhenReceiving))

(event Custom|PopUp%20Event
  (params (Instigator :type Controller%20Object%20Reference))
  ;[0.3658 0 1 0.25] Gameplay Functions
  (CallFunction|PopUpwhenReceiving))

(event Custom|PopUpWhenReceiving
  ;[0.3658 0 1 0.25] Gameplay Functions
  (if (and (not (Variables|Default|GetTargetIsUp)) (CallFunction|TargetEnabled))
    (Variables|Default|SetWaitingForProximity false)
    (Utilities|Time|ClearandInvalidateTimerbyHandle (Variables|Default|GetPopUpDelayTimer))
    (Utilities|Time|ClearandInvalidateTimerbyHandle (Variables|Default|GetHideTargetTimer))
    (CallFunction|PopUpTarget true)))

(event Custom|PopDown%20Event
  (params (Instigator :type Controller%20Object%20Reference))
  ;[0.3658 0 1 0.25] Gameplay Functions
  (CallFunction|KnockdownwhenReceiving))

(event Custom|HopUp%20Event
  (params (Instigator :type Controller%20Object%20Reference))
  ;[0.3658 0 1 0.25] Gameplay Functions
  (CallFunction|HopUpwhenReceiving))

(event Custom|KnockdownWhenReceiving
  ;[0.3658 0 1 0.25] Gameplay Functions
  (if (and (Variables|Default|GetTargetIsUp) (CallFunction|TargetEnabled))
    (Variables|Default|SetWaitingForProximity false)
    (Utilities|Time|ClearandInvalidateTimerbyHandle (Variables|Default|GetHideTargetTimer))
    (Utilities|Time|ClearandInvalidateTimerbyHandle (Variables|Default|GetPopUpDelayTimer))
    (CallFunction|HideTarget)))

(event Custom|DropDown%20Event
  (params (Instigator :type Controller%20Object%20Reference))
  ;[0.3658 0 1 0.25] Gameplay Functions
  (CallFunction|DropDownwhenReceiving))

(event Custom|HopUpWhenReceiving
  ;[0.3658 0 1 0.25] Gameplay Functions
  (if (and (not (Variables|Default|GetInJumpUpPosition)) (CallFunction|TargetEnabled))
    (Variables|Default|SetWaitingForProximity false)
    (Variables|Default|SetInJumpUpPosition true)))

(event Custom|DropDownWhenReceiving
  ;[0.3658 0 1 0.25] Gameplay Functions
  (if (and (Variables|Default|GetInJumpUpPosition) (CallFunction|TargetEnabled))
    (Variables|Default|SetWaitingForProximity false)
    (Variables|Default|SetInJumpUpPosition false)))

(event OnEnabledStateChanged(Creative_Enabled_Component)
  (params (Enabled :type Boolean))
  (if (not Enabled)
    (CallFunction|KnockdownTarget false)))

(event EventOnDeathPlayEffects
  (params (Damage :type Float%20%28single-precision%29) (DamageTags :type Gameplay%20Tag%20Container%20Structure%20%28by%20ref%29) (Momentum :type Vector :default "0, 0, 0") (HitInfo :type Hit%20Result%20Structure%20%28by%20ref%29) (InstigatedBy :type Fort%20Pawn%20Object%20Reference) (DamageCauser :type Actor%20Object%20Reference) (EffectContext :type Gameplay%20Effect%20Context%20Handle%20Structure))
  ;[1 1 1] So Damage Numbers continue to show even if the target has "lost all of its health"
  (CallFunction|ClearPlayedDying))

(fn AddTargetTransmitDispatcher
  (params (NewHitEventListener :type Target%20Hit%20Event%20Broadcast%20Interface%20Interface))
  (Utilities|Array|AddUnique (Variables|Default|GetTargetTransmitListeners) NewHitEventListener))

(fn AdjustDamageBasedOnNumberDisplay
  (params (InDamage :type Float%20%28double-precision%29) (BullseyeHit :type Boolean))
  ;[1 1 1] Not being used currently.  Will revisit if we use the "Number display settings"
  (Variables|Default|SetCachedDamage InDamage)
  (switch int (Variables|PlayerOptions|GetNumberDisplaySetting)
    (:0
      (LocalVariables|SetDamage 0.0))
    (:1
      (LocalVariables|SetDamage (Variables|GetInDamage)))
    (:2
      (if (Variables|GetBullseyeHit)
        (LocalVariables|SetDamage (Variables|GetInDamage))
        (else
          (LocalVariables|SetDamage 0.0))))
    (:3
      (bind _scoretoaward (CallFunction|CalculateScore false))
      (LocalVariables|SetDamage (Math|Conversions|ToFloat(Integer) _scoretoaward))))
  (bind _output_get (Variables|Default|SetTotalDamageDealt (LocalVariables|GetDamage)))
  (return _output_get))

(fn AwardScore
  (params (FortPlayerController :type Fort%20Player%20Controller%20Object%20Reference))
  (bind (_outsourcetags _outcontexttags) (Class|FortMinigameStatsComponent|GetSourceandContextTags (LocalVariables|Getlocal%20MinigameStatsComp) (LocalVariables|Getlocal%20FortPlayerController)))
  (LocalVariables|Setlocal%20FortPlayerController FortPlayerController)
  (LocalVariables|Getlocal%20FortPlayerController
    (:then
      (if (or (== (Teams|TeamtoIndex (Team|GetActorTeam FortPlayerController)) (Knobs|Team|GetTeamValueasIndex (Variables|PlayerOptions|GetTeamAssignmentSetting))) (Knobs|Team|IsTeamAll (Variables|PlayerOptions|GetTeamAssignmentSetting)))
        (bind _scoretoaward (CallFunction|CalculateScore true))
        (bind _returnvalue (CreativeDevices|GetMinigameStatsComp (LocalVariables|Getlocal%20FortPlayerController) "Athena"))
        (LocalVariables|Setlocal%20MinigameStatsComp _returnvalue)
        (bind _local_minigamestatscomp (LocalVariables|Getlocal%20MinigameStatsComp)
          (:then
            (Class|FortMinigameStatsComponent|SendStatEventwithTags _local_minigamestatscomp (LocalVariables|Getlocal%20FortPlayerController) "MinigameDynamicEvent" self (GameplayTags|MakeGameplayTagContainerfromTag "(TagName=\"Minigame.Stat.Score.ShootingTarget\")") _outsourcetags _outcontexttags _scoretoaward)
            (return _scoretoaward (Variables|PlayerOptions|GetTeamAssignmentSetting)))
          (:else))))
    (:else)))

(fn BlueprintModifyIncomingDamage
  (params (Damage :type Float%20%28single-precision%29) (InTags :type Gameplay%20Tag%20Container%20Structure%20%28by%20ref%29) (EffectContext :type Gameplay%20Effect%20Context%20Handle%20Structure%20%28by%20ref%29) (EventInstigator :type Controller%20Object%20Reference) (DamageCauser :type Actor%20Object%20Reference))
  (bind _returnvalue (|CallParentFunction Damage InTags EffectContext EventInstigator DamageCauser))
  (LocalVariables|SetNewDamage _returnvalue)
  (if (CallFunction|TargetEnabled)
    (Utilities|FlowControl|Sequence
      (:then_0
        (if (Ability|EffectContext|HasHitResult EffectContext)
          (bind (_hitcomponent _returnvalue_2) (CallFunction|WeakPointWasHit (Ability|EffectContext|GetHitResult EffectContext)))
          (if _returnvalue_2
            (bind _outweapondata (CallFunction|GetWeaponDatafromDamageCauser DamageCauser))
            (bind _outdamage (CallFunction|CalculateDamagefromWeaponData _outweapondata (LocalVariables|GetNewDamage)))
            (LocalVariables|SetNewDamage _outdamage))))
      (:then_1
        _))
    (else
      (LocalVariables|SetNewDamage 0.0)))
  (return (LocalVariables|GetNewDamage)))

(fn CalculateDamageFromWeaponData
  (params (WeaponData :type Fort%20Weapon%20Item%20Definition%20Object%20Reference) (InDamage :type Float%20%28double-precision%29))
  (bind _weapondata (Variables|GetWeaponData))
  (LocalVariables|SetNewDamage InDamage)
  (Utilities|FlowControl|Sequence
    (:then_0
      (Utilities|IsValid _weapondata
        (:Is%20Valid
          (bind (_outrow _returnvalue) (Fortnite|GetWeaponStatsRow (WeaponItemDefinition|GetWeaponStatHandle _weapondata)))
          (if _returnvalue
            (LocalVariables|SetNewDamage (* (LocalVariables|GetNewDamage) (.DamageZone_Critical (Utilities|Struct|BreakFortBaseWeaponStats _outrow))))))))
    (:then_1
      (return (LocalVariables|GetNewDamage)))))

(fn CalculateScore
  (params (Knockdown :type Boolean))
  (if Knockdown
    _
    (elif (not (Variables|Default|GetScoreOnKnockdownsOnly))
      _
      (else
        (return 0))))
  (if (Variables|Default|GetWeakPointHit)
    (return (* (Variables|PlayerOptions|GetScoreValueSetting) (Variables|PlayerOptions|GetWeakPointMultiplierSetting)))
    (else
      (return (Variables|PlayerOptions|GetScoreValueSetting)))))

(fn CanPopUp
  (params (MinigameJustStartedOrStopped :type Boolean) (WasShot :type Boolean))
  (bind _minigamejuststartedorstopped (Variables|GetMinigameJustStartedOrStopped))
  ;[1 1 1] Checks if the target should pop up based on current state and the Max Resets and Max Pop Ups Settings values
  (if (Variables|Default|GetTargetIsUp)
    (return false)
    (elif (Math|Boolean|ANDBoolean (== (Variables|PlayerOptions|GetMaxPopUpsSetting) 0) (== (Variables|PlayerOptions|GetMaxResetsSetting) 0) (not _minigamejuststartedorstopped))
      (return false)
      (elif (Variables|GetWasShot)
        (if (== (Variables|PlayerOptions|GetMaxResetsSetting) 99)
          _
          (elif (< (Variables|Default|GetNumResets) (Variables|PlayerOptions|GetMaxResetsSetting))
            _
            (else
              (return false))))
        (return true)
        (elif (or _minigamejuststartedorstopped (== (Variables|PlayerOptions|GetMaxPopUpsSetting) 99))
          (return true)
          (elif (< (Variables|Default|GetNumPopUps) (Variables|PlayerOptions|GetMaxPopUpsSetting))
            _
            (else
              (return false))))))))

(fn CheckIfShouldHop
  (if (!= (Variables|PlayerOptions|GetJumpTimeSetting) 0.0)
    (Networking|SwitchHasAuthority
      (:Authority
        (if (Variables|Default|GetInJumpUpPosition)
          (CallFunction|SelectHoporDropTime false)
          (else
            (CallFunction|SelectHoporDropTime true)))))
    (else
      (CallFunction|ClearBunnyHopTimer) ; Setting is off, make sure hopping is stopped immediately
    )))

(fn CheckRandomMinMax
  (if (> (Variables|PlayerOptions|GetMinResetTime) (Variables|PlayerOptions|GetMaxResetTime))
    (Variables|PlayerOptions|SetMinResetTime 1.0)
    (Variables|PlayerOptions|SetMaxResetTime 6e+02)
    (else
      _))
  (if (> (Variables|PlayerOptions|GetMinPopupTime) (Variables|PlayerOptions|GetMaxPopupTime))
    (Variables|PlayerOptions|SetMinPopupTime 1.0)
    (Variables|PlayerOptions|SetMaxPopupTime 6e+02)
    (else
      _))
  (if (> (Variables|PlayerOptions|GetMinHideTime) (Variables|PlayerOptions|GetMaxHideTime))
    (Variables|PlayerOptions|SetMinHideTime 1.0)
    (Variables|PlayerOptions|SetMaxHideTime 6e+02)
    (else
      _))
  (if (> (Variables|PlayerOptions|GetMinHoppedTime) (Variables|PlayerOptions|GetMaxHoppedTime))
    (Variables|PlayerOptions|SetMinHoppedTime 0.5)
    (Variables|PlayerOptions|SetMaxHoppedTime 2e+01)
    (else
      _))
  (if (> (Variables|PlayerOptions|GetMinTimeToHop) (Variables|PlayerOptions|GetMaxTimeToHop))
    (Variables|PlayerOptions|SetMinTimeToHop 0.5)
    (Variables|PlayerOptions|SetMaxTimeToHop 2e+01)))

(fn ClearAllTimers
  (Utilities|Time|ClearandInvalidateTimerbyHandle (Variables|Default|GetPopUpDelayTimer))
  (Utilities|Time|ClearandInvalidateTimerbyHandle (Variables|Default|GetHopTimer))
  (Utilities|Time|ClearandInvalidateTimerbyHandle (Variables|Default|GetHoppedUpPositionTimer))
  (Utilities|Time|ClearandInvalidateTimerbyHandle (Variables|Default|GetHideTargetTimer)))

(fn ClearBunnyHopTimer
  ;[1 1 1] Clear Hop timer
  (Utilities|Time|ClearandInvalidateTimerbyHandle (Variables|Default|GetHopTimer)))

(fn GetCollisionStaticMesh
  (return "0, 0, 0" (Variables|/Script/Engine.StaticMeshComponent|GetStaticMesh (Variables|Default|GetSM_Target_Base))))

(fn GetMeshComponents
  (return 0))

(fn GetWeaponDataFromDamageCauser
  (params (DamageCauser :type Object%20Reference))
  (bind _damagecauser (Variables|GetDamageCauser))
  (Utilities|FlowControl|Sequence
    (:then_0
      (bind _asfort_weapon (Utilities|Casting|CastToFortWeapon _damagecauser)
        (:then
          (LocalVariables|SetWeaponData (Data|GetWeaponData _asfort_weapon)))
        (:CastFailed
          (bind _asfort_projectile_athena (Utilities|Casting|CastToFortProjectileAthena _damagecauser))
          (bind _asfort_weapon_2 (Utilities|Casting|CastToFortWeapon (Actor|GetOwner _asfort_projectile_athena)))
          (LocalVariables|SetWeaponData (Data|GetWeaponData _asfort_weapon_2)))))
    (:then_1
      (return (LocalVariables|GetWeaponData)))))

(fn HandleHitTransmitting
  (params (Bullseye :type Boolean) (Knockdown :type Boolean) (TriggerInstigator :type Controller%20Object%20Reference))
  (if Bullseye
    (bind _sentsomething (CallFunction|HandleTransmitter (Variables|GetTriggerInstigator) "NewEnumerator3"))
    (if _sentsomething
      _
      (elif Knockdown
        (bind _sentsomething_2 (CallFunction|HandleTransmitter (Variables|GetTriggerInstigator) "NewEnumerator2"))
        (if (not _sentsomething_2)
          (CallFunction|HandleTransmitter (Variables|GetTriggerInstigator) "NewEnumerator1"))
        (else
          _)))
    (else
      _)))

(fn HandleTransmitter
  (params (TriggerInstigator :type Controller%20Object%20Reference) (Event :type E_ShootingTargetTransmitter%20Enum :default "NewEnumerator0"))
  (if (Class|CreativeCommonDeviceFunctionLibrary|IsinPlayMode self)
    (if (Variables|Default|GetTrackAttached)
      (for _array_element (Variables|Default|GetTargetTransmitListeners)
        (bind _bsenttoanything (Class|TargetHitEventBroadcastInterface|HandleTargetHitEventBroadcasts _array_element (Variables|GetTriggerInstigator) (Variables|GetEvent)))
        (LocalVariables|SetSentSomethingThrough (or _bsenttoanything (LocalVariables|GetSentSomethingThrough))))
      (return (LocalVariables|GetSentSomethingThrough))
      (else
        (switch Utilities|FlowControl|Switch|SwitchonE_ShootingTargetTransmitter Event
          (:NewEnumerator1
            (bind _returnvalue (GameplayEvent|BroadcastEvent (Variables|UserOptions|Events|GetOn%20Hit) 0 (Utilities|Struct|MakeGameplayEventonShootingTargetHit (Variables|GetTriggerInstigator))))
            (return _returnvalue))
          (:NewEnumerator2
            (bind _returnvalue_2 (GameplayEvent|BroadcastEvent (Variables|UserOptions|Events|GetOn%20Knockdown) 0 (Utilities|Struct|MakeGameplayEventonShootingTargetKnockdown (Variables|GetTriggerInstigator))))
            (return _returnvalue_2))
          (:NewEnumerator3
            (bind _returnvalue_3 (GameplayEvent|BroadcastEvent (Variables|UserOptions|Events|GetOn%20Bullseye%20Hit) 0 (Utilities|Struct|MakeGameplayEventonShootingTargetBullseye (Variables|GetTriggerInstigator))))
            (return _returnvalue_3))
          (:NewEnumerator4
            (bind _returnvalue_4 (GameplayEvent|BroadcastEvent (Variables|UserOptions|Events|GetOn%20Pop-Up) 0 (Utilities|Struct|MakeGameplayEventonShootingTargetPopup (Variables|GetTriggerInstigator))))
            (return _returnvalue_4))
          (:NewEnumerator5
            (bind _returnvalue_5 (GameplayEvent|BroadcastEvent (Variables|UserOptions|Events|GetOn%20Pop-Down) 0 (Utilities|Struct|MakeGameplayEventonShootingTargetPopdown (Variables|GetTriggerInstigator))))
            (return _returnvalue_5))
          (:NewEnumerator6
            (bind _returnvalue_6 (GameplayEvent|BroadcastEvent (Variables|UserOptions|Events|GetOn%20Hop-Up) 0 (Utilities|Struct|MakeGameplayEventonShootingTargetHopup (Variables|GetTriggerInstigator))))
            (return _returnvalue_6))
          (:NewEnumerator7
            (bind _returnvalue_7 (GameplayEvent|BroadcastEvent (Variables|UserOptions|Events|GetOn%20Hop-Down) 0 (Utilities|Struct|MakeGameplayEventonShootingTargetHopdown (Variables|GetTriggerInstigator))))
            (return _returnvalue_7)))))))

(fn HideTarget
  (if (Variables|Default|GetTargetIsUp)
    (Networking|SwitchHasAuthority
      (:Authority
        (CallFunction|SelectResetorPopupTime false)
        (CallFunction|RPCMulticastKnockdownTarget false)))))

(fn IgnoreWorldGridScaleOverride
  (return false))

(fn InitializeHideTargetTimer
  ;[1 1 1] If the Hide target setting is on, and the target is already up, clear and then set the timer.
  (if (and (!= (Variables|PlayerOptions|GetTimeBeforeHidingSetting) 0.0) (Variables|Default|GetTargetIsUp))
    (Utilities|Time|ClearandInvalidateTimerbyHandle (Variables|Default|GetHideTargetTimer))
    (bind _output_time (CallFunction|SelectHideTime))
    (CallFunction|SetHideTargetTimer _output_time)))

(fn IsWithinWeakpointBounds
  (params (Impact%20location :type Vector) (Component :type Scene%20Component%20Object%20Reference))
  (bind (_output_targetname_8_3ccb565a4b3f684a6c49d78ad7818216 _output_location_2_bd1586e3401487fe676d8c9a1ba6b438 _output_radius_5_fd7a33fb431d79189b659589b1e3e1e7) (Utilities|Struct|BreakShootingTargetWeakPointDetails (Utilities|Array|Get(acopy) (Variables|Default|GetWeakpointDetails) (Variables|PlayerOptions|GetTargetMeshSetting))))
  ;[1 1 1] Takes the hit location and sees if it falls within the bounds of the weak point collision component.  This is necessary because projectiles don't register as hitting collision components.
  (return (<= (Math|Vector|VectorLengthSquared (- (- Impact%20location (Variables|Default|GetActorLocationAtMomentofHit)) _output_location_2_bd1586e3401487fe676d8c9a1ba6b438)) (Math|Float|Square (* (* _output_radius_5_fd7a33fb431d79189b659589b1e3e1e7 2.0) (Variables|PlayerOptions|GetBullseyeSizeSetting))))))

(fn MinigameIsRunning
  (bind _fortminigameprogress (Variables|Default|GetFortMinigameProgress))
  (bind _currentminigame (Variables|/Script/FortniteGame.FortMinigameLogicComponent|GetCurrentMinigame _fortminigameprogress))
  (Utilities|IsValid _currentminigame
    (:Is%20Valid
      (return (Minigame|HasStarted _currentminigame :!class /Script/FortniteGame.FortMinigame)))))

(fn OnRep_DropTime
  (bind _droptime (Variables|Default|GetDropTime))
  ;[1 1 1] If Hop Length isn't set to forever, set the timer to drop
  (if (not (== _droptime 0))
    (CallFunction|SetDropHopTimer _droptime)))

(fn OnRep_HopTime
  (CallFunction|SetHopTimer (Variables|Default|GetHopTime)))

(fn OnRep_InJumpUpPosition
  (if (Variables|Default|GetInJumpUpPosition)
    (Utilities|Time|ClearandInvalidateTimerbyHandle (Variables|Default|GetHoppedUpPositionTimer))
    (CallFunction|Hop)
    (else
      (Utilities|Time|ClearandInvalidateTimerbyHandle (Variables|Default|GetHopTimer))
      (CallFunction|DropfromHop))))

(fn OnRep_ResetDelaySetting)

(fn OnRep_SavedResetDelaySetting)

(fn OnRep_TargetIsDisabled)

(fn OnRep_TargetMeshSetting
  (CallFunction|PickMeshfromSettings))

(fn PostDamageApplyHealthAdjustment
  (params (InDamage :type Float%20%28double-precision%29))
  ;[1 1 1] Not being used currently.  Will revisit if we use the "Number display settings"
  ;;[1 1 1] Is it invincible?
  (if (== (Variables|PlayerOptions|GetStartingHealthSetting) (Variables|Default|GetInfiniteHealthIndex))
    (LocalVariables|SetHealth (Variables|PlayerOptions|GetStartingHealthSetting))
    (else
      (LocalVariables|SetHealth (- (+ InDamage (Health|GetHealth)) (Variables|Default|GetCachedDamage)))))
  (Health|SetHealth (LocalVariables|GetHealth)))

(fn ReadyToActivate
  (return (and (CallFunction|TargetEnabled) (not (Variables|Default|GetWaitingForProximity)))))

(fn RemoveTargetTransmitDispatcher
  (params (HitEventListenerToRemove :type Target%20Hit%20Event%20Broadcast%20Interface%20Interface))
  (Utilities|Array|RemoveItem (Variables|Default|GetTargetTransmitListeners) HitEventListenerToRemove))

(fn RestoreToMaxHealth
  (Networking|SwitchHasAuthority
    (:Authority
      (Health|SetHealth (Health|GetMaxHealth)))))

(fn SelectDropTime
  (if (CallFunction|ReadytoActivate)
    (bind _output_get (LocalVariables|SetTime (Math|Float|SelectFloat (Math|Random|RandomFloatinRange (Variables|PlayerOptions|GetMinHideTime) (Variables|PlayerOptions|GetMaxHideTime)) (Variables|PlayerOptions|GetTimeBeforeHidingSetting) (== (Variables|PlayerOptions|GetTimeBeforeHidingSetting) -1.0))))
    (return _output_get)
    (else
      (return 0.0))))

(fn SelectHideTime
  (if (CallFunction|ReadytoActivate)
    (bind _output_get (LocalVariables|SetTime (Math|Float|SelectFloat (Math|Random|RandomFloatinRange (Variables|PlayerOptions|GetMinHideTime) (Variables|PlayerOptions|GetMaxHideTime)) (Variables|PlayerOptions|GetTimeBeforeHidingSetting) (== (Variables|PlayerOptions|GetTimeBeforeHidingSetting) -1.0))))
    (return _output_get)
    (else
      (return 0.0))))

(fn SelectHopOrDropTime
  (params (HoppingUp :type Boolean))
  (Networking|SwitchHasAuthority
    (:Authority
      (if HoppingUp
        (if (== -1.0 (Variables|PlayerOptions|GetJumpTimeSetting))
          (Variables|Default|SetHopTime (Math|Random|RandomFloatinRange (Variables|PlayerOptions|GetMinTimeToHop) (Variables|PlayerOptions|GetMaxTimeToHop)))
          (else
            (Variables|Default|SetHopTime (Variables|PlayerOptions|GetJumpTimeSetting))))
        (elif (== -1.0 (Variables|PlayerOptions|GetHopLengthSetting))
          (Variables|Default|SetDropTime (Math|Random|RandomFloatinRange (Variables|PlayerOptions|GetMinHoppedTime) (Variables|PlayerOptions|GetMaxHoppedTime)))
          (else
            (Variables|Default|SetDropTime (Variables|PlayerOptions|GetHopLengthSetting)))))
      (return (LocalVariables|GetTime)))))

(fn SelectResetOrPopupTime
  (params (Was%20Shot :type Boolean))
  (bind _was_shot (Variables|GetWas%20Shot))
  (bind _popupdelaysetting (Variables|PlayerOptions|GetPopupDelaySetting))
  (bind _resetdelaysetting (Variables|PlayerOptions|GetResetDelaySetting))
  (bind _returnvalue (Math|Float|SelectFloat _resetdelaysetting _popupdelaysetting _was_shot))
  (if (CallFunction|ReadytoActivate)
    (LocalVariables|SetTime _returnvalue)
    ;[1 1 1] Was the time set to random?
    (if (== _returnvalue -1.0)
      (LocalVariables|SetTime (Math|Float|SelectFloat (Math|Random|RandomFloatinRange (Variables|PlayerOptions|GetMinResetTime) (Variables|PlayerOptions|GetMaxResetTime)) (Math|Random|RandomFloatinRange (Variables|PlayerOptions|GetMinPopupTime) (Variables|PlayerOptions|GetMaxPopupTime)) (Variables|GetWas%20Shot)))
      (else
        _))
    (Variables|Default|SetPopup%20Time (LocalVariables|GetTime))
    (return)
    (else
      (LocalVariables|SetTime 0.0)
      (return))))

(fn SetAimAssistEnabled
  (params (NewType :type Boolean))
  (bind _returnvalue (select NewType "ECR_Block" "ECR_Ignore"))
  (bind _ascharacter (Utilities|Casting|CastToCharacter (Variables|/Script/Engine.ChildActorComponent|GetChildActor (Variables|Default|GetAimAssistPawn))))
  (Utilities|IsValid (Variables|/Script/Engine.Character|GetCapsuleComponent _ascharacter)
    (:Is%20Valid
      (Collision|SetCollisionResponsetoChannel (Variables|/Script/Engine.Character|GetCapsuleComponent _ascharacter) "ECC_GameTraceChannel10" _returnvalue)
      (Collision|SetCollisionResponsetoChannel (Variables|/Script/Engine.Character|GetCapsuleComponent _ascharacter) "ECC_GameTraceChannel3" _returnvalue)
      (Development|PrintText (Utilities|Text|FormatText "NSLOCTEXT(\"[997D23F04F0DAC2112C424850FD0872A]\", \"1B3EBD83430797CAC0AFC1AC2272969B\", \"PawnCapsule {0} WeaponCapsule {1}\")" (Utilities|Text|ToText(Byte) (Collision|GetCollisionResponsetoChannel (Variables|/Script/Engine.Character|GetCapsuleComponent _ascharacter) "ECC_GameTraceChannel10")) (Utilities|Text|ToText(Byte) (Collision|GetCollisionResponsetoChannel (Variables|/Script/Engine.Character|GetCapsuleComponent _ascharacter) "ECC_GameTraceChannel3"))) true true "(R=0.000000,G=0.660000,B=1.000000,A=1.000000)" 5.0))))

(fn SetBunnyHopTimer
  ;[1 1 1] If set to bunny hop, set timers based on current position of the target
  (Networking|SwitchHasAuthority
    (:Authority
      (if (CallFunction|ReadytoActivate)
        (if (!= (Variables|PlayerOptions|GetJumpTimeSetting) 0.0)
          (if (Variables|Default|GetInJumpUpPosition)
            (CallFunction|SelectHoporDropTime false)
            (else
              (CallFunction|SelectHoporDropTime true))))))))

(fn SetEnabledStateCheck
  (params (StartMatch :type Boolean))
  (Class|CreativeEnabledComponent|SetEnabled (Variables|Default|GetCreative_Enabled_Component) (and StartMatch (Variables|PlayerOptions|GetEnabledOnMinigameStart))))

(fn SetLegacyResetDefaults
  (bind _toyoptionscomponent (Variables|Default|GetToyOptionsComponent))
  (bind _currentoptions (OptionData|GetPropertyOverrides _toyoptionscomponent))
  (bind _value (Utilities|Map|Find _currentoptions "ResetDelaySetting"))
  ;[1 1 1] Will this always overwrite if it is set to default, or only if the setting has never changed?  Either way if the player never changes the one value will it constantly overwrite?
  (if (Variables|Default|GetFirstLoad)
    (Variables|Default|SetFirstLoad false)
    (if (Utilities|String|NotEqualExactly(String) _value (Variables|Default|GetDefaultResetDelayValue))
      (OptionData|SetPropertyOverrideValues _toyoptionscomponent (Utilities|Map|MakeMap "PopupDelaySetting" _value))
      (CallFunction|Resave))))

(fn SetMaxHealth
  (params (NewMaxHealth :type Float%20%28double-precision%29))
  (bind _self self)
  (bind _returnvalue (Ability|GetAbilitySystemComponent _self))
  ;[1 1 1] -1.0 means they are invincible.  We will prevent damage in the "BlueprintModifyIncomingDamage" Function
  (Networking|SwitchHasAuthority
    (:Authority
      (if (== NewMaxHealth (Variables|Default|GetInfiniteHealthIndex))
        (LocalVariables|SetLocalNewMaxHealth 1.0)
        (else
          (LocalVariables|SetLocalNewMaxHealth NewMaxHealth)))
      (bind _returnvalue_2 (Ability|GameplayEffect|AssignTagSetbyCallerMagnitude (GameplayEffects|MakeOutgoingSpec :self (Variables|Default|GetAbilitySystem) :GameplayEffectClass (Variables|Default|GetModifyMaxHealth) :Level 0.0) (Variables|Default|GetSetMaxHealthDataTag) (LocalVariables|GetLocalNewMaxHealth)))
      (Utilities|IsValid _returnvalue
        (:Is%20Valid
          (GameplayEffects|RemoveActiveGameplayEffect _returnvalue (Variables|Default|GetMaxHealthEffectHandler) -1)
          (bind _returnvalue_3 (GameplayEffects|ApplyGameplayEffectSpecToSelf _returnvalue _returnvalue_2))
          (Variables|Default|SetMaxHealthEffectHandler _returnvalue_3)
          (Health|SetHealth (Health|GetMaxHealth)) ; Don't let it actually die.  Set health to 1% so bar looks mostly empty
        )))))

(fn SetStartingPosition
  (if (Variables|Default|GetInJumpUpPosition)
    (Variables|Default|SetStartOfMatchHopDown true)
    (Variables|Default|SetInJumpUpPosition false)
    (else
      _))
  (Networking|SwitchHasAuthority
    (:Authority
      (if (== (Variables|PlayerOptions|GetStartingPositionSetting) 1)
        (if (CallFunction|ReadytoActivate)
          (CallFunction|SelectResetorPopupTime false)
          (else
            _))
        (CallFunction|RPCMulticastKnockdownTarget false)
        (elif (Variables|Default|GetTargetIsUp)
          _
          (else
            (CallFunction|PopUpTarget true)))))))

(fn ShouldTriggerTarget
  (params (Overlapping%20Actor :type Object%20Reference) (Overlapping%20Component :type Primitive%20Component%20Object%20Reference))
  ;[1 1 1] Ensures that overlapping actor should trigger the target.  Currently handles players and occupied vehicles.  Also ensures minigame is running.
  (bind _asplayer_pawn_athena (Utilities|Casting|CastToPlayerPawn_Athena Overlapping%20Actor)
    (:then
      (return (not (Vehicle|IsinVehicle _asplayer_pawn_athena))))
    (:CastFailed
      (bind _asfort_athena_vehicle (Utilities|Casting|CastToFortAthenaVehicle Overlapping%20Actor)
        (:then
          (return (> (Utilities|Array|Length (Vehicle|GetAllPassengers _asfort_athena_vehicle)) 0)))
        (:CastFailed
          (return false))))))

(fn TargetEnabled
  (return (Class|CreativeEnabledComponent|IsEnabled (Variables|Default|GetCreative_Enabled_Component))))

(fn TargetShouldFall
  ;[1 1 1] Checks health and weakpoint knockdown setting to determine if target should fall
  (return (or (<= (Health|GetHealth) 0.0) (and (Variables|Default|GetWeakPointHit) (Variables|Default|GetWeakPointKnockdownSetting)))))

(fn UpdateProximitySphereCollision
  (if (<= (Variables|PlayerOptions|GetProximityDistanceSetting) (Variables|Default|GetProximityOffSetting))
    (Collision|SetCollisionEnabled (Variables|Default|GetProximityTrigger) "NoCollision")
    (else
      (Collision|SetCollisionEnabled (Variables|Default|GetProximityTrigger) "QueryOnly")
      (Components|Sphere|SetSphereRadius (Variables|Default|GetProximityTrigger) (Math|Conversions|ToFloat(Integer) (Variables|PlayerOptions|GetProximityDistanceSetting)) true) ; Sets size of prox trigger based on setting
    )))

(fn UpdateHealthBarVisibility
  (params (ShouldShowBar :type Boolean))
  (CallFunction|SetSuppressHealthBar (or (not ShouldShowBar) (Variables|Default|GetUserSuppressHealthBar)))
  (return))

(fn ConstructionScript
  (CallFunction|PickMeshfromSettings))

(fn WeakPointWasHit
  (params (Hit%20Result :type Hit%20Result%20Structure%20%28by%20ref%29))
  (bind (_bblockinghit _binitialoverlap _time _distance _location _impactpoint _normal _impactnormal _physmat _hitactor _hitcomponent) (Collision|BreakHitResult Hit%20Result))
  (bind _returnvalue (CallFunction|IsWithinWeakpointBounds _impactpoint _hitcomponent))
  (if _returnvalue
    (CallFunction|SpawnWeakpointFX 0 (Math|Transform|MakeTransform _impactpoint (Math|Conversions|RotationFromXVector _normal)))
    (Variables|Default|SetWeakPointHit true)
    (return 0 true)
    (else
      (Variables|Default|SetWeakPointHit false)
      (return _hitcomponent false))))
