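; B_TargetTrack is a Creative device that slides an attached shooting target along a length-scalable track.
; An authority-set, replicated movement state drives a timeline forward, reverse or stopped, with play rate compensating for track length.
; Receiver events, proximity overlaps, looping and the target's knockdown/pop-up dispatchers change that state.
; The track passes options and channels to its child target and rebroadcasts target hits as gameplay events.
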(event MinigameStartedDelegate(FortMinigameProgress)
  (bind _targetmovement (Variables|B_TargetTrack|GetTargetMovement))
  ;[1 1 1] Game State Changes
  (if (not (Variables|Default|GetActiveAtGameStart))
    (Components|Timeline|Stop _targetmovement)
    (Components|Timeline|SetPlaybackPosition _targetmovement 0.0 false true)
    (CallFunction|SetMovementState (Variables|Default|GetStopMovementState))
    (Variables|Default|SetIsActive false)))

(event OnComponentBeginOverlap(Box)
  (params (OverlappedComponent :type Primitive%20Component%20Object%20Reference) (OtherActor :type Actor%20Object%20Reference) (OtherComp :type Primitive%20Component%20Object%20Reference) (OtherBodyIndex :type Integer) (bFromSweep :type Boolean) (SweepResult :type Hit%20Result%20Structure%20%28by%20ref%29))
  (bind _initialmovementdelaysetting (Variables|Default|GetInitialMovementDelaySetting))
  ;[1 1 1] Proximity trigger to start movement
  (Networking|SwitchHasAuthority
    (:Authority
      (if (and (Variables|Default|GetTrackEnabled) (!= (Variables|Default|GetProximityTriggerSizeSetting) (Variables|Default|GetProximityOffValue)))
        (if (and (not (Variables|Default|GetIsCurrentlyMoving)) (Variables|Default|GetAtStartPoint))
          (Utilities|FlowControl|Delay _initialmovementdelaysetting)
          (CallFunction|SetMovementState (Variables|Default|GetLeftToRightmovementState))
          (elif (and (not (Variables|Default|GetIsCurrentlyMoving)) (not (Variables|Default|GetAtStartPoint)))
            (Utilities|FlowControl|Delay _initialmovementdelaysetting)
            (CallFunction|SetMovementState (Variables|Default|GetRightToLeftMovementState))))))))

(event MinigameStateChangedDelegate(FortMinigameProgress)
  (params (Minigame :type Fort%20Minigame%20Object%20Reference) (NewMinigameState :type EFortMinigameState%20Enum :default "PreGame"))
  ;[1 1 1] Enable / Disable player detection so the On Component Begin Overlap (Box) is not called during the transitions
  (if (!= (Variables|Default|GetProximityTriggerSizeSetting) (Variables|Default|GetProximityOffValue))
    (Collision|SetCollisionEnabled (Variables|Default|GetBox) (select (or (Utilities|Enum|Equal(Enum) NewMinigameState "InProgress") (Utilities|Enum|Equal(Enum) NewMinigameState "PreGame")) "QueryOnly" "NoCollision"))))

(event EventBeginPlay
  (bind _attachedshootingtarget (Variables|Default|GetAttachedShootingTarget))
  (Variables|/Script/Engine.SceneComponent|SetbAbsoluteScale true (Variables|Default|GetBox))
  ;[1 1 1] Attach Target to track
  (Networking|SwitchHasAuthority
    (:Authority
      (bind _asb_shooting_target_master (Utilities|Casting|CastToB_ShootingTarget_Master (Variables|/Script/Engine.ChildActorComponent|GetChildActor (Variables|Default|GetEditorTarget))))
      (bind _output_get (Variables|Default|SetAttachedShootingTarget _asb_shooting_target_master))
      (Variables|/Game/Creative/Devices/ShootingTargets/B_ShootingTarget_Master.B_ShootingTarget_Master_C|SetTrackAttached true _output_get)
      ;[1 1 1] Dispatchers are in B_ShootingTarget_Master and are called when the target is knocked down or pops up.  These events allow the track to change its behavior based on the target's state.
      (Default|BindEventtoTargetKnockedDownDispatcher _attachedshootingtarget (delegate Custom|AttachedTargetKnockedDown))
      (Default|BindEventtoTargetPopUpDispatcher _attachedshootingtarget (delegate Custom|AttachedTargetPoppedUp)))))

(event Custom|AttachedTargetPoppedUp
  ;[1 1 1] Dispatchers are in B_ShootingTarget_Master and are called when the target is knocked down or pops up.  These events allow the track to change its behavior based on the target's state.
  (if (Variables|Default|GetIsActive)
    (if (or (Variables|Default|GetLoopingSetting) (Variables|Default|GetIsCurrentlyMoving))
      (if (Variables|Default|GetAtStartPoint)
        (CallFunction|SetMovementState (Variables|Default|GetLeftToRightmovementState))
        (else
          (CallFunction|SetMovementState (Variables|Default|GetRightToLeftMovementState)))))))

(event Custom|AttachedTargetKnockedDown
  ;[1 1 1] Dispatchers are in B_ShootingTarget_Master and are called when the target is knocked down or pops up.  These events allow the track to change its behavior based on the target's state.
  (CallFunction|SetMovementState (Variables|Default|GetStopMovementState)))

(event EventEndPlay
  (params (EndPlayReason :type EEndPlayReason%20Enum :default "Destroyed"))
  ;[1 1 1] Destroy attached target when deleted
  (CallFunction|StopMoveTargetAudio))

(event EventDestroyed
  ;[1 1 1] Destroy attached target when deleted
  (CallFunction|StopMoveTargetAudio))

(event Custom|MoveTarget
  (bind _attachedshootingtarget (Variables|Default|GetAttachedShootingTarget))
  ;[1 1 1] Target movement
  (CallFunction|MoveTargetAudio)
  (bind (_direction _newtrack_0) (Timeline|TargetMovement Play)
    (:Update
      (Variables|Default|SetIsCurrentlyMoving true)
      (CallFunction|UpdateLerpedPosition _newtrack_0))
    (:Finished
      (CallFunction|StopMoveTargetAudio)
      (if (Utilities|Enum|Equal(Enum) _direction "Forward")
        (Variables|Default|SetAtStartPoint false)
        (else
          (Variables|Default|SetAtStartPoint true)))
      (Variables|Default|SetIsCurrentlyMoving false)
      (if (Variables|Default|GetLoopingSetting)
        (Utilities|FlowControl|Delay (Variables|Default|GetDelayAtEndSetting))
        (if (Utilities|IsValid _attachedshootingtarget :!class /Script/Engine.KismetSystemLibrary)
          (if (Math|Boolean|ANDBoolean (Variables|Default|GetTrackEnabled) (Variables|Default|GetIsActive) (Variables|/Game/Creative/Devices/ShootingTargets/B_ShootingTarget_Master.B_ShootingTarget_Master_C|GetTargetIsUp _attachedshootingtarget))
            (if (Utilities|Enum|Equal(Enum) _direction "Forward")
              (CallFunction|SetMovementState (Variables|Default|GetRightToLeftMovementState))
              (else
                (CallFunction|SetMovementState (Variables|Default|GetLeftToRightmovementState))))))))))

(event Custom|StopTargetMovement
  (bind _attachedshootingtarget (Variables|Default|GetAttachedShootingTarget))
  ;[1 1 1] Target movement
  (CallFunction|StopMoveTargetAudio)
  (Variables|Default|SetIsCurrentlyMoving false)
  (Timeline|TargetMovement Stop))

(event Custom|Stop%20Move%20Target%20Audio
  ;[1 1 1] Audio for when Target is moving on track
  (Audio|Components|Audio|Play (Variables|Default|GetTarget_Move_Stop_Cue) 0.0)
  (Audio|Components|Audio|Stop (Variables|Default|GetTarget_Move_Loop_Cue)))

(event Custom|ReverseTargetMovement
  (bind _attachedshootingtarget (Variables|Default|GetAttachedShootingTarget))
  ;[1 1 1] Target movement
  (CallFunction|MoveTargetAudio)
  (Timeline|TargetMovement Reverse))

(event Custom|Move%20Target%20Audio
  ;[1 1 1] Audio for when Target is moving on track
  (Utilities|IsValid (Variables|Default|GetAttachedShootingTarget)
    (:Is%20Valid
      (Audio|Components|Audio|Play (Variables|Default|GetTarget_Move_Start_Cue) 0.0)
      (Audio|Components|Audio|FadeIn (Variables|Default|GetTarget_Move_Loop_Cue) 0.5 1.0 0.0 "Linear" :!class /Script/Engine.AudioComponent))))

(event Custom|Restart
  (bind _attachedshootingtarget (Variables|Default|GetAttachedShootingTarget))
  ;[1 1 1] Gets everything happy when options are set
  (if (Math|Boolean|ANDBoolean (Variables|Default|GetLoopingSetting) (not (Variables|Default|GetAtStartPoint)) (not (Variables|Default|GetIsCurrentlyMoving)))
    ;[1 1 1] Target movement
    (Utilities|FlowControl|Delay (Variables|Default|GetDelayAtEndSetting))
    (if (Utilities|IsValid _attachedshootingtarget :!class /Script/Engine.KismetSystemLibrary)
      (if (Math|Boolean|ANDBoolean (Variables|Default|GetTrackEnabled) (Variables|Default|GetIsActive) (Variables|/Game/Creative/Devices/ShootingTargets/B_ShootingTarget_Master.B_ShootingTarget_Master_C|GetTargetIsUp _attachedshootingtarget))
        (if (Utilities|Enum|Equal(Enum) (stale Custom|MoveTarget _direction) "Forward")
          (CallFunction|SetMovementState (Variables|Default|GetRightToLeftMovementState))
          (else
            (CallFunction|SetMovementState (Variables|Default|GetLeftToRightmovementState))))))
    (elif (and (Variables|Default|GetLoopingSetting) (not (Variables|Default|GetIsCurrentlyMoving)))
      (CallFunction|SetMovementState (Variables|Default|GetLeftToRightmovementState)))))

(event OnPropertiesLoaded(ToyOptionsComponent)
  (bind _attachedshootingtarget (Variables|Default|GetAttachedShootingTarget))
  (Utilities|FlowControl|DelayUntilNextTick)
  (Variables|Default|GetAttachedShootingTarget
    (:then
      _)
    (:else
      (bind _asb_shooting_target_master (Utilities|Casting|CastToB_ShootingTarget_Master (Variables|/Script/Engine.ChildActorComponent|GetChildActor (Variables|Default|GetEditorTarget))))
      ;[1 1 1] Pass our property overrides to our attached target
      (Variables|Default|SetAttachedShootingTarget _asb_shooting_target_master)))
  (CallFunction|MergeTrackandTargetOptions)
  (Class|BShootingTargetMaster|UpdateProximitySphereCollision (Variables|Default|GetAttachedShootingTarget))
  (CallFunction|InitializeTargetChannels)
  ;[1 1 1] Gets everything happy when options are set
  (CallFunction|SetTargetRotationandBaseMesh)
  (CallFunction|SetTrackLength)
  (CallFunction|SetupProximityVolume)
  (bind _returnvalue (CallFunction|CalculatePlayRate))
  (Components|Timeline|SetPlayRate (Variables|B_TargetTrack|GetTargetMovement) _returnvalue) ; Keeps movement rate consistent regardless of length of track
  (CallFunction|SetAudioSpeedSound _returnvalue)
  (if (Math|Boolean|ANDBoolean (Variables|Default|GetLoopingSetting) (not (Variables|Default|GetAtStartPoint)) (not (Variables|Default|GetIsCurrentlyMoving)))
    ;[1 1 1] Target movement
    (Utilities|FlowControl|Delay (Variables|Default|GetDelayAtEndSetting))
    (if (Utilities|IsValid _attachedshootingtarget :!class /Script/Engine.KismetSystemLibrary)
      (if (Math|Boolean|ANDBoolean (Variables|Default|GetTrackEnabled) (Variables|Default|GetIsActive) (Variables|/Game/Creative/Devices/ShootingTargets/B_ShootingTarget_Master.B_ShootingTarget_Master_C|GetTargetIsUp _attachedshootingtarget))
        (if (Utilities|Enum|Equal(Enum) (stale Custom|MoveTarget _direction) "Forward")
          (CallFunction|SetMovementState (Variables|Default|GetRightToLeftMovementState))
          (else
            (CallFunction|SetMovementState (Variables|Default|GetLeftToRightmovementState))))))
    (elif (and (Variables|Default|GetLoopingSetting) (not (Variables|Default|GetIsCurrentlyMoving)))
      (CallFunction|SetMovementState (Variables|Default|GetLeftToRightmovementState)))))

(event OnAnyPropertyChanged(ToyOptionsComponent)
  (bind _attachedshootingtarget (Variables|Default|GetAttachedShootingTarget))
  (Utilities|FlowControl|DelayUntilNextTick)
  (Variables|Default|GetAttachedShootingTarget
    (:then
      _)
    (:else
      (bind _asb_shooting_target_master (Utilities|Casting|CastToB_ShootingTarget_Master (Variables|/Script/Engine.ChildActorComponent|GetChildActor (Variables|Default|GetEditorTarget))))
      ;[1 1 1] Pass our property overrides to our attached target
      (Variables|Default|SetAttachedShootingTarget _asb_shooting_target_master)))
  (CallFunction|MergeTrackandTargetOptions)
  (Class|BShootingTargetMaster|UpdateProximitySphereCollision (Variables|Default|GetAttachedShootingTarget))
  (CallFunction|InitializeTargetChannels)
  ;[1 1 1] Gets everything happy when options are set
  (CallFunction|SetTargetRotationandBaseMesh)
  (CallFunction|SetTrackLength)
  (CallFunction|SetupProximityVolume)
  (bind _returnvalue (CallFunction|CalculatePlayRate))
  (Components|Timeline|SetPlayRate (Variables|B_TargetTrack|GetTargetMovement) _returnvalue) ; Keeps movement rate consistent regardless of length of track
  (CallFunction|SetAudioSpeedSound _returnvalue)
  (if (Math|Boolean|ANDBoolean (Variables|Default|GetLoopingSetting) (not (Variables|Default|GetAtStartPoint)) (not (Variables|Default|GetIsCurrentlyMoving)))
    ;[1 1 1] Target movement
    (Utilities|FlowControl|Delay (Variables|Default|GetDelayAtEndSetting))
    (if (Utilities|IsValid _attachedshootingtarget :!class /Script/Engine.KismetSystemLibrary)
      (if (Math|Boolean|ANDBoolean (Variables|Default|GetTrackEnabled) (Variables|Default|GetIsActive) (Variables|/Game/Creative/Devices/ShootingTargets/B_ShootingTarget_Master.B_ShootingTarget_Master_C|GetTargetIsUp _attachedshootingtarget))
        (if (Utilities|Enum|Equal(Enum) (stale Custom|MoveTarget _direction) "Forward")
          (CallFunction|SetMovementState (Variables|Default|GetRightToLeftMovementState))
          (else
            (CallFunction|SetMovementState (Variables|Default|GetLeftToRightmovementState))))))
    (elif (and (Variables|Default|GetLoopingSetting) (not (Variables|Default|GetIsCurrentlyMoving)))
      (CallFunction|SetMovementState (Variables|Default|GetLeftToRightmovementState)))))

(event Custom|InitializeSettings
  (bind _attachedshootingtarget (Variables|Default|GetAttachedShootingTarget))
  (Utilities|FlowControl|DelayUntilNextTick)
  (Variables|Default|GetAttachedShootingTarget
    (:then
      _)
    (:else
      (bind _asb_shooting_target_master (Utilities|Casting|CastToB_ShootingTarget_Master (Variables|/Script/Engine.ChildActorComponent|GetChildActor (Variables|Default|GetEditorTarget))))
      ;[1 1 1] Pass our property overrides to our attached target
      (Variables|Default|SetAttachedShootingTarget _asb_shooting_target_master)))
  (CallFunction|MergeTrackandTargetOptions)
  (Class|BShootingTargetMaster|UpdateProximitySphereCollision (Variables|Default|GetAttachedShootingTarget))
  (CallFunction|InitializeTargetChannels)
  ;[1 1 1] Gets everything happy when options are set
  (CallFunction|SetTargetRotationandBaseMesh)
  (CallFunction|SetTrackLength)
  (CallFunction|SetupProximityVolume)
  (bind _returnvalue (CallFunction|CalculatePlayRate))
  (Components|Timeline|SetPlayRate (Variables|B_TargetTrack|GetTargetMovement) _returnvalue) ; Keeps movement rate consistent regardless of length of track
  (CallFunction|SetAudioSpeedSound _returnvalue)
  (if (Math|Boolean|ANDBoolean (Variables|Default|GetLoopingSetting) (not (Variables|Default|GetAtStartPoint)) (not (Variables|Default|GetIsCurrentlyMoving)))
    ;[1 1 1] Target movement
    (Utilities|FlowControl|Delay (Variables|Default|GetDelayAtEndSetting))
    (if (Utilities|IsValid _attachedshootingtarget :!class /Script/Engine.KismetSystemLibrary)
      (if (Math|Boolean|ANDBoolean (Variables|Default|GetTrackEnabled) (Variables|Default|GetIsActive) (Variables|/Game/Creative/Devices/ShootingTargets/B_ShootingTarget_Master.B_ShootingTarget_Master_C|GetTargetIsUp _attachedshootingtarget))
        (if (Utilities|Enum|Equal(Enum) (stale Custom|MoveTarget _direction) "Forward")
          (CallFunction|SetMovementState (Variables|Default|GetRightToLeftMovementState))
          (else
            (CallFunction|SetMovementState (Variables|Default|GetLeftToRightmovementState))))))
    (elif (and (Variables|Default|GetLoopingSetting) (not (Variables|Default|GetIsCurrentlyMoving)))
      (CallFunction|SetMovementState (Variables|Default|GetLeftToRightmovementState)))))

(event Custom|On%20AttachedShootingTarget%20Loaded
  (bind _attachedshootingtarget (Variables|Default|GetAttachedShootingTarget))
  (Utilities|FlowControl|DelayUntilNextTick)
  (Variables|Default|GetAttachedShootingTarget
    (:then
      _)
    (:else
      (bind _asb_shooting_target_master (Utilities|Casting|CastToB_ShootingTarget_Master (Variables|/Script/Engine.ChildActorComponent|GetChildActor (Variables|Default|GetEditorTarget))))
      ;[1 1 1] Pass our property overrides to our attached target
      (Variables|Default|SetAttachedShootingTarget _asb_shooting_target_master)))
  (CallFunction|MergeTrackandTargetOptions)
  (Class|BShootingTargetMaster|UpdateProximitySphereCollision (Variables|Default|GetAttachedShootingTarget))
  (CallFunction|InitializeTargetChannels)
  ;[1 1 1] Gets everything happy when options are set
  (CallFunction|SetTargetRotationandBaseMesh)
  (CallFunction|SetTrackLength)
  (CallFunction|SetupProximityVolume)
  (bind _returnvalue (CallFunction|CalculatePlayRate))
  (Components|Timeline|SetPlayRate (Variables|B_TargetTrack|GetTargetMovement) _returnvalue) ; Keeps movement rate consistent regardless of length of track
  (CallFunction|SetAudioSpeedSound _returnvalue)
  (if (Math|Boolean|ANDBoolean (Variables|Default|GetLoopingSetting) (not (Variables|Default|GetAtStartPoint)) (not (Variables|Default|GetIsCurrentlyMoving)))
    ;[1 1 1] Target movement
    (Utilities|FlowControl|Delay (Variables|Default|GetDelayAtEndSetting))
    (if (Utilities|IsValid _attachedshootingtarget :!class /Script/Engine.KismetSystemLibrary)
      (if (Math|Boolean|ANDBoolean (Variables|Default|GetTrackEnabled) (Variables|Default|GetIsActive) (Variables|/Game/Creative/Devices/ShootingTargets/B_ShootingTarget_Master.B_ShootingTarget_Master_C|GetTargetIsUp _attachedshootingtarget))
        (if (Utilities|Enum|Equal(Enum) (stale Custom|MoveTarget _direction) "Forward")
          (CallFunction|SetMovementState (Variables|Default|GetRightToLeftMovementState))
          (else
            (CallFunction|SetMovementState (Variables|Default|GetLeftToRightmovementState))))))
    (elif (and (Variables|Default|GetLoopingSetting) (not (Variables|Default|GetIsCurrentlyMoving)))
      (CallFunction|SetMovementState (Variables|Default|GetLeftToRightmovementState)))))

(event Custom|AsyncLoadTargetBase
  (bind _leftrighttargetbasemesh (Variables|Default|GetLeftRightTargetBaseMesh))
  (bind _output (Utilities|ResolveSoftReference _leftrighttargetbasemesh))
  (bind _forwardbackwardtargetbasemesh (Variables|Default|GetForwardBackwardTargetBaseMesh))
  (bind _output_2 (Utilities|ResolveSoftReference _forwardbackwardtargetbasemesh))
  (bind _attachedshootingtarget (Variables|Default|GetAttachedShootingTarget))
  (bind _sm_target_base (Variables|/Game/Creative/Devices/ShootingTargets/B_ShootingTarget_Master.B_ShootingTarget_Master_C|GetSM_Target_Base _attachedshootingtarget))
  (bind _attachedshootingtarget_2 (Variables|Default|GetAttachedShootingTarget))
  (bind _sm_target_base_2 (Variables|/Game/Creative/Devices/ShootingTargets/B_ShootingTarget_Master.B_ShootingTarget_Master_C|GetSM_Target_Base _attachedshootingtarget_2))
  (if (<= (Variables|Default|GetRotationSetting) 1)
    ;[1 1 1] Set base to left/right config
    (if (Utilities|IsValidSoftObjectReference _leftrighttargetbasemesh)
      (Utilities|IsValid _output
        (:Is%20Valid
          (Components|StaticMesh|SetStaticMesh _sm_target_base_2 _output))
        (:Is%20Not%20Valid
          (bind _loaded_object (Creative|Loading|AsyncLoadEditorSafe _leftrighttargetbasemesh)
            (:Loaded
              (bind _asstatic_mesh (Utilities|Casting|CastToStaticMesh _loaded_object))
              (Components|StaticMesh|SetStaticMesh _sm_target_base_2 _asstatic_mesh))))))
    (elif (Utilities|IsValidSoftObjectReference _forwardbackwardtargetbasemesh)
      ;[1 1 1] Set base to forward backward configuration
      (Utilities|IsValid _output_2
        (:Is%20Valid
          (Components|StaticMesh|SetStaticMesh _sm_target_base _output_2))
        (:Is%20Not%20Valid
          (bind _loaded_object_2 (Creative|Loading|AsyncLoadEditorSafe _forwardbackwardtargetbasemesh)
            (:Loaded
              (bind _asstatic_mesh_2 (Utilities|Casting|CastToStaticMesh _loaded_object_2))
              (Components|StaticMesh|SetStaticMesh _sm_target_base _asstatic_mesh_2))))))))

(event Custom|EnableFromEvent
  (params (Instigator :type Controller%20Object%20Reference))
  ;[1 1 1] Receivers
  (if (Class|CreativeCommonDeviceFunctionLibrary|IsinPlayMode self)
    (Class|BShootingTargetMaster|EnableTarget (Variables|Default|GetAttachedShootingTarget))))

(event Custom|DisableFromEvent
  (params (Instigator :type Controller%20Object%20Reference))
  ;[1 1 1] Receivers
  (if (Class|CreativeCommonDeviceFunctionLibrary|IsinPlayMode self)
    (Class|BShootingTargetMaster|DisableTarget (Variables|Default|GetAttachedShootingTarget))))

(event Custom|ActivateWhenReceivingEvent
  (params (Instigator :type Controller%20Object%20Reference))
  ;[1 1 1] Receivers
  (if (Variables|Default|GetTrackEnabled)
    ;;[1 1 1] Set to the previous movement state, unless it was stopped completely, then just set it to move forward
    (CallFunction|SetMovementState (select (Math|Byte|Equal(Byte) (Variables|Default|GetCachedMovementState) 0) 1 (Variables|Default|GetCachedMovementState)))
    (Variables|Default|SetIsActive true)))

(event Custom|DeactivateWhenReceivingEvent
  (params (Instigator :type Controller%20Object%20Reference))
  ;[1 1 1] Receivers
  (Variables|Default|SetCachedMovementState (Variables|Default|GetCurrentMovementState))
  ;;[1 1 1] Set to the previous movement state, unless it was stopped completely, then just set it to move forward
  (CallFunction|SetMovementState (Variables|Default|GetStopMovementState))
  (Variables|Default|SetIsActive false))

(event Custom|MoveToEndEvent
  (params (Instigator :type Controller%20Object%20Reference))
  ;[1 1 1] Receivers
  (if (Variables|Default|GetTrackEnabled)
    (CallFunction|SetMovementState (Variables|Default|GetLeftToRightmovementState))
    (Variables|Default|SetIsActive true)))

(event Custom|MoveToStartEvent
  (params (Instigator :type Controller%20Object%20Reference))
  ;[1 1 1] Receivers
  (if (Variables|Default|GetTrackEnabled)
    (CallFunction|SetMovementState (Variables|Default|GetRightToLeftMovementState))
    (Variables|Default|SetIsActive true)))

(event Custom|DisableMovementEvent
  (params (Instigator :type Controller%20Object%20Reference))
  ;[1 1 1] Receivers
  (Variables|Default|SetCachedMovementState (Variables|Default|GetCurrentMovementState))
  (CallFunction|SetMovementState (Variables|Default|GetStopMovementState))
  (Variables|Default|SetIsActive false)
  (Variables|Default|SetTrackEnabled false))

(event Custom|EnableMovementEvent
  (params (Instigator :type Controller%20Object%20Reference))
  ;[1 1 1] Receivers
  (Variables|Default|SetTrackEnabled true))

(event Custom|KnockDownWhenReceivingEvent
  (params (Instigator :type Controller%20Object%20Reference))
  (bind _attachedshootingtarget (Variables|Default|GetAttachedShootingTarget))
  ;[1 1 1] Receivers
  (Utilities|IsValid _attachedshootingtarget
    (:Is%20Valid
      (Class|BShootingTargetMaster|KnockdownwhenReceiving _attachedshootingtarget))))

(event Custom|HopUpWhenReceivingEvent
  (params (Instigator :type Controller%20Object%20Reference))
  (bind _attachedshootingtarget (Variables|Default|GetAttachedShootingTarget))
  ;[1 1 1] Receivers
  (Utilities|IsValid _attachedshootingtarget
    (:Is%20Valid
      (Class|BShootingTargetMaster|HopUpwhenReceiving _attachedshootingtarget))))

(event Custom|DropDownWhenReceivingEvent
  (params (Instigator :type Controller%20Object%20Reference))
  (bind _attachedshootingtarget (Variables|Default|GetAttachedShootingTarget))
  ;[1 1 1] Receivers
  (Utilities|IsValid _attachedshootingtarget
    (:Is%20Valid
      (Class|BShootingTargetMaster|DropDownwhenReceiving _attachedshootingtarget))))

(event Custom|ResetWhenReceivingEvent
  (params (Instigator :type Controller%20Object%20Reference))
  (bind _attachedshootingtarget (Variables|Default|GetAttachedShootingTarget))
  ;[1 1 1] Receivers
  (Utilities|IsValid _attachedshootingtarget
    (:Is%20Valid
      (Class|BShootingTargetMaster|ResetwhenReceiving _attachedshootingtarget))))

(event Custom|PopUpWhenReceivingEvent
  (params (Instigator :type Controller%20Object%20Reference))
  (bind _attachedshootingtarget (Variables|Default|GetAttachedShootingTarget))
  ;[1 1 1] Receivers
  (Utilities|IsValid _attachedshootingtarget
    (:Is%20Valid
      (Class|BShootingTargetMaster|PopUpwhenReceiving _attachedshootingtarget))))

(fn BTargetTrackHandleTransmitters
  (params (Instigator :type Controller%20Object%20Reference) (Event :type E_ShootingTargetTransmitter%20Enum :default "NewEnumerator0")))

(fn Play
  (params (bUnpaused :type Boolean))
  (bind _attachedshootingtarget (Variables|Default|GetAttachedShootingTarget))
  (bind _bfailedtoplay (|CallParentFunction bUnpaused))
  (if (and (not _bfailedtoplay) (not bUnpaused))
    (Utilities|IsValid _attachedshootingtarget
      (:Is%20Valid
        (Class|BShootingTargetMaster|AddTargetTransmitDispatcher _attachedshootingtarget self))
      (:Is%20Not%20Valid
        _)))
  (return _bfailedtoplay))

(fn Stop
  (bind _attachedshootingtarget (Variables|Default|GetAttachedShootingTarget))
  (bind _bfailedtostop (|CallParentFunction))
  (Utilities|IsValid _attachedshootingtarget
    (:Is%20Valid
      (Class|BShootingTargetMaster|RemoveTargetTransmitDispatcher _attachedshootingtarget self))
    (:Is%20Not%20Valid
      _))
  (return _bfailedtostop))

(fn CalculatePlayRate
  ;[1 1 1] Accounts for settings but keeps the movement speed of the target over 512 units uniform
  (return (* (/ 1.0 (Variables|Default|GetLengthScaleSetting)) (Variables|Default|GetTrackSpeedSetting))))

(fn EditorUpdates
  (bind _asb_shooting_target_master (Utilities|Casting|CastToB_ShootingTarget_Master (Variables|/Script/Engine.ChildActorComponent|GetChildActor (Variables|Default|GetEditorTarget))))
  (Variables|/Game/Creative/Devices/ShootingTargets/B_ShootingTarget_Master.B_ShootingTarget_Master_C|SetTargetMeshSetting (Variables|PlayerOptions|GetTargetMeshSetting) _asb_shooting_target_master)
  (CallFunction|SetTargetRotationandBaseMesh)
  (CallFunction|SetTrackLength))

(fn GetCollisionStaticMesh
  (return "0, 0, 0" (Variables|/Script/Engine.StaticMeshComponent|GetStaticMesh (Variables|Default|GetSM_Target_Track))))

(fn GetFinalDestinationOfTarget
  (return (Math|Vector|MakeVector (* (Variables|Default|GetLengthScaleSetting) 512.0))))

(fn GetMeshComponents
  (return 0))

(fn HandleTargetHitEventBroadcasts
  (params (Instigator :type Controller%20Object%20Reference) (Event :type E_ShootingTargetTransmitter%20Enum :default "NewEnumerator0"))
  (if (Class|CreativeCommonDeviceFunctionLibrary|IsinPlayMode self)
    (switch Utilities|FlowControl|Switch|SwitchonE_ShootingTargetTransmitter Event
      (:NewEnumerator1
        (bind _returnvalue (GameplayEvent|BroadcastEvent (Variables|UserOptions|Events|GetOn%20Hit) 0 (Utilities|Struct|MakeGameplayEventonTargetTrackTargetHit (Variables|Actor|GetInstigator))))
        (return _returnvalue))
      (:NewEnumerator2
        (bind _returnvalue_2 (GameplayEvent|BroadcastEvent (Variables|UserOptions|Events|GetOn%20Knockdown) 0 (Utilities|Struct|MakeGameplayEventonTargetTrackKnockdown (Variables|Actor|GetInstigator))))
        (return _returnvalue_2))
      (:NewEnumerator3
        (bind _returnvalue_3 (GameplayEvent|BroadcastEvent (Variables|UserOptions|Events|GetOn%20Bullseye%20Hit) 0 (Utilities|Struct|MakeGameplayEventonTargetTrackBullseye (Variables|Actor|GetInstigator))))
        (return _returnvalue_3))
      (:NewEnumerator4
        (bind _returnvalue_4 (GameplayEvent|BroadcastEvent (Variables|UserOptions|Events|GetOn%20Pop-Up) 0 (Utilities|Struct|MakeGameplayEventonTargetTrackPopup (Variables|Actor|GetInstigator))))
        (return _returnvalue_4))
      (:NewEnumerator5
        (bind _returnvalue_5 (GameplayEvent|BroadcastEvent (Variables|UserOptions|Events|GetOn%20Pop-Down) 0 (Utilities|Struct|MakeGameplayEventonTargetTrackPopdown (Variables|Actor|GetInstigator))))
        (return _returnvalue_5))
      (:NewEnumerator6
        (bind _returnvalue_6 (GameplayEvent|BroadcastEvent (Variables|UserOptions|Events|GetOn%20Hop-Up) 0 (Utilities|Struct|MakeGameplayEventonTargetTrackHopup (Variables|Actor|GetInstigator))))
        (return _returnvalue_6))
      (:NewEnumerator7
        (bind _returnvalue_7 (GameplayEvent|BroadcastEvent (Variables|UserOptions|Events|GetOn%20Hop-Down) 0 (Utilities|Struct|MakeGameplayEventonTargetTrackHopdown (Variables|Actor|GetInstigator))))
        (return _returnvalue_7)))))

(fn IgnoreWorldGridScaleOverride
  (return false))

(fn InitializeTargetChannels
  (bind _attachedshootingtarget (Variables|Default|GetAttachedShootingTarget))
  (Networking|SwitchHasAuthority
    (:Authority
      (Class|FortGameplayMessageComponentBase|SetChannelId (Variables|/Game/Creative/Devices/ShootingTargets/B_ShootingTarget_Master.B_ShootingTarget_Master_C|GetHitTransmitter _attachedshootingtarget) (Class|FortGameplayMessageComponentBase|GetChannelId (Variables|Default|GetHitTransmitter)))
      (Class|FortGameplayMessageComponentBase|SetChannelId (Variables|/Game/Creative/Devices/ShootingTargets/B_ShootingTarget_Master.B_ShootingTarget_Master_C|GetOnHitKnockdownTransmitOn _attachedshootingtarget) (Class|FortGameplayMessageComponentBase|GetChannelId (Variables|Default|GetOnHitKnockdownTransmitOn)))
      (Class|FortGameplayMessageComponentBase|SetChannelId (Variables|/Game/Creative/Devices/ShootingTargets/B_ShootingTarget_Master.B_ShootingTarget_Master_C|GetOnBullseyeTransmitOn _attachedshootingtarget) (Class|FortGameplayMessageComponentBase|GetChannelId (Variables|Default|GetOnBullseyeTransmitOn)))
      (Class|FortGameplayMessageComponentBase|SetChannelId (Variables|/Game/Creative/Devices/ShootingTargets/B_ShootingTarget_Master.B_ShootingTarget_Master_C|GetOnPopUpTransmitOn _attachedshootingtarget) (Class|FortGameplayMessageComponentBase|GetChannelId (Variables|Default|GetOnPopUpTransmitOn)))
      (Class|FortGameplayMessageComponentBase|SetChannelId (Variables|/Game/Creative/Devices/ShootingTargets/B_ShootingTarget_Master.B_ShootingTarget_Master_C|GetOnPopDownTransmitOn _attachedshootingtarget) (Class|FortGameplayMessageComponentBase|GetChannelId (Variables|Default|GetOnPopDownTransmitOn)))
      (Class|FortGameplayMessageComponentBase|SetChannelId (Variables|/Game/Creative/Devices/ShootingTargets/B_ShootingTarget_Master.B_ShootingTarget_Master_C|GetOnHopUpTransmitOn _attachedshootingtarget) (Class|FortGameplayMessageComponentBase|GetChannelId (Variables|Default|GetOnHopUpTransmitOn)))
      (Class|FortGameplayMessageComponentBase|SetChannelId (Variables|/Game/Creative/Devices/ShootingTargets/B_ShootingTarget_Master.B_ShootingTarget_Master_C|GetOnHopDownTransmitOn _attachedshootingtarget) (Class|FortGameplayMessageComponentBase|GetChannelId (Variables|Default|GetOnHopDownTransmitOn))))))

(fn MergeTrackAndTargetOptions
  (bind _attachedshootingtarget (Variables|Default|GetAttachedShootingTarget))
  ;[1 1 1] Pass our property overrides to our attached target
  (if (and (Utilities|IsValid _attachedshootingtarget :!class /Script/Engine.KismetSystemLibrary) (Networking|HasAuthority))
    (OptionData|SetPropertyOverrides (Variables|/Game/Creative/Devices/ShootingTargets/B_ShootingTarget_Master.B_ShootingTarget_Master_C|GetToyOptionsComponent _attachedshootingtarget) (OptionData|GetPropertyOverrides (Variables|Default|GetToyOptionsComponent)))))

(fn OnRep_AttachedShootingTarget
  (CallFunction|OnAttachedShootingTargetLoaded))

(fn OnRep_CurrentDirection
  (CallFunction|UpdateforRepNotify))

(fn OnRep_CurrentLerpValue)

(fn OnRep_CurrentMovementState
  (switch int (Math|Conversions|ToInteger(Byte) (Variables|Default|GetCurrentMovementState))
    (:Default
      (CallFunction|StopTargetMovement))
    (:0
      (CallFunction|StopTargetMovement))
    (:1
      (CallFunction|MoveTarget) ; Left to Right movement
    )
    (:2
      (CallFunction|ReverseTargetMovement) ; Right to Left movement
    )))

(fn OnRep_IsCurrentlyMoving
  (CallFunction|UpdateforRepNotify))

(fn OnRep_IsTargetUp
  (CallFunction|UpdateforRepNotify))

(fn OnRep_ServerMoveStartTime
  (CallFunction|UpdateforRepNotify))

(fn SetAudioSpeedSound
  (params (Target%20Speed :type Float%20%28double-precision%29))
  (Utilities|FlowControl|Sequence
    (:then_0
      (if (<= Target%20Speed 0.5)
        (Variables|Audio|SetAudio%20Speed%20Parameter 0)))
    (:then_1
      (if (and (> Target%20Speed 0.5) (< Target%20Speed 1.0))
        (Variables|Audio|SetAudio%20Speed%20Parameter 1)))
    (:then_2
      (if (>= Target%20Speed 1.0)
        (Variables|Audio|SetAudio%20Speed%20Parameter 2))))
  (Audio|Parameter|SetIntegerParameter (Variables|Default|GetTarget_Move_Loop_Cue) "TargetSpeed" (Variables|Audio|GetAudio%20Speed%20Parameter) :!class /Script/AudioExtensions.AudioParameterControllerInterface))

(fn SetMovementState
  (params (State :type Byte))
  (Networking|SwitchHasAuthority
    (:Authority
      (if (Math|Byte|NotEqual(Byte) State (Variables|Default|GetCurrentMovementState))
        (Variables|Default|SetCurrentMovementState State)))))

(fn SetTargetRotationAndBaseMesh
  (bind _targetattachpoint (Variables|Default|GetTargetAttachPoint))
  (switch int (Variables|Default|GetRotationSetting)
    (:Default
      (Transformation|SetRelativeRotation _targetattachpoint "0, 0, 0" false false :!label _label) ; Backward
    )
    (:0
      (Transformation|SetRelativeRotation _targetattachpoint "0,90.000000, 0" false false) ; Right
    )
    (:1
      (Transformation|SetRelativeRotation _targetattachpoint "0,-90.000000, 0" false false) ; Left
    )
    (:2
      (Transformation|SetRelativeRotation _targetattachpoint "0,180.000000, 0" false false) ; Forward
    )
    (:3
      (goto _label)))
  (Utilities|IsValid (Variables|Default|GetAttachedShootingTarget)
    (:Is%20Valid
      (CallFunction|AsyncLoadTargetBase))))

(fn SetTrackLength
  (Transformation|SetRelativeScale3D (Variables|Default|GetSM_Target_Track) (Math|Vector|MakeVector (Variables|Default|GetLengthScaleSetting) 1.0 1.0)))

(fn SetupProximityVolume
  (bind _box (Variables|Default|GetBox))
  (bind (_relativelocation_x _relativelocation_y) (Math|Vector|BreakVector (Variables|/Script/Engine.SceneComponent|GetRelativeLocation_X _box)))
  (bind (_boxextent_x _boxextent_y _boxextent_z) (Math|Vector|BreakVector (Variables|/Script/Engine.BoxComponent|GetBoxExtent_X _box)))
  (if (!= (Variables|Default|GetProximityTriggerSizeSetting) (Variables|Default|GetProximityOffValue))
    (Components|Box|SetBoxExtent _box (Math|Vector|MakeVector (+ (* (Variables|Default|GetProximityTriggerSizeSetting) 512.0) (* (Variables|Default|GetLengthScaleSetting) 512.0)) (* (Variables|Default|GetProximityTriggerSizeSetting) 512.0) (/ (* (Variables|Default|GetProximityTriggerSizeSetting) 512.0) 2.0)) true)
    (Transformation|SetRelativeLocation _box (Math|Vector|MakeVector _relativelocation_x _relativelocation_y _boxextent_z) false false)))

(fn UpdateForRepNotify)

(fn UpdateLerpedPosition
  (params (Alpha :type Float%20%28double-precision%29))
  (bind _destination (CallFunction|GetFinalDestinationOfTarget))
  (Transformation|SetRelativeLocation (Variables|Default|GetTargetAttachPoint) (Math|Vector|Lerp(Vector) "0, 0, 0" _destination Alpha) false false))

(fn UpdateTargetSettings
  (bind _attachedshootingtarget (Variables|Default|GetAttachedShootingTarget)
    (:then
      (UserOptionValues|SetUserOptionValues _attachedshootingtarget 0 (UserOptionValues|GetUserOptionValues)))
    (:else)))

(fn ConstructionScript
  (|CallParentFunction)
  (if (Editor|InEditor)
    (CallFunction|EditorUpdates)))
