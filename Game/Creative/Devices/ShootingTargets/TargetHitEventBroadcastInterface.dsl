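; TargetHitEventBroadcastInterface is an interface for Creative shooting target devices to broadcast target hit events.
; It declares one function, HandleTargetHitEventBroadcasts, which takes an instigating Controller and a shooting target transmitter enum event.
; The function has no implementation here and shows only an empty default return, leaving the handling to implementers.
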(fn HandleTargetHitEventBroadcasts
  (params (Instigator :type Controller%20Object%20Reference) (Event :type E_ShootingTargetTransmitter%20Enum :default "NewEnumerator0"))
  (return false))
